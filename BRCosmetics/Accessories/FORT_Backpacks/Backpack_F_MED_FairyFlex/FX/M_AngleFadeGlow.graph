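# Unlit, additive, two-sided glow material for Niagara sprite and mesh particles in the FairyFlex backpack effects.
# A UV-centered SphereMask with adjustable radius and hardness, tinted by particle color, feeds MF_FX_EmissiveAdjustments for emissive.
# The same mask, multiplied by a view-angle fade and particle alpha, forms opacity.
# A quality switch adds depth fade, skipping it on Low and Medium.
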Material /BRCosmetics/Accessories/FORT_Backpacks/Backpack_F_MED_FairyFlex/FX/M_AngleFadeGlow
graph {
 properties {
  bAutomaticallySetUsageInEditor=false
  bUsedWithNiagaraMeshParticles=true
  bUsedWithNiagaraSprites=true
  blendMode=BLEND_Additive
  pixelDepthOffsetMode=PDOM_Legacy
  shadingModel=MSM_Unlit
  twoSided=true
  usedShadingModels=Unlit
 }
 nodes [
  n1 = ParticleColor()
  n2 = SphereMask(
   A=TextureCoordinate()
   B=Constant() {r=0.5}
   Radius=ScalarParameter() {defaultValue=0.5, parameterName=Radius}
   Hardness=ScalarParameter() {parameterName=Hardness}
  )
  MaterialFunctionCall(Emissive=Multiply(A=@n1.RGB, B=@n2)) {materialFunction=&MF_FX_EmissiveAdjustments}
  n3 = ScalarParameter() {parameterName=Fade}
  n4 = Multiply(
   A=Multiply(
    A=@n2
    B=MaterialFunctionCall("FadeSensitivity (optional)"=@n3) {materialFunction=&FadeBasedOnViewAngle}.Result
   )
   B=@n1.A
  )
  QualitySwitch(Default=DepthFade(Opacity=@n4, FadeDistance=ScalarParameter() {defaultValue=10, parameterName=DepthFade}), Low=@n4, Medium=@n4)
 ]
}
