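# M_FresnelGlow is an unlit additive Fresnel glow material for Niagara mesh particles in the FairyFlex backpack cosmetic effects.
# MF_Fresnel_FX, tuned by Brightness and Exponent parameters, multiplies particle color RGB, which then passes through MF_FX_EmissiveAdjustments.
# Particle alpha, Fresnel and vertex color multiply into an opacity value.
# A quality switch applies parameterized depth fade to that value by default, bypassing it on Low and Medium.
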Material /BRCosmetics/Accessories/FORT_Backpacks/Backpack_F_MED_FairyFlex/FX/M_FresnelGlow
graph {
 properties {
  bAutomaticallySetUsageInEditor=false
  bUsedWithNiagaraMeshParticles=true
  blendMode=BLEND_Additive
  pixelDepthOffsetMode=PDOM_Legacy
  shadingModel=MSM_Unlit
  usedShadingModels=Unlit
 }
 nodes [
  n1 = ParticleColor()
  n2 = MaterialFunctionCall(Brightness=ScalarParameter() {parameterName=Brightness}, Exponent=ScalarParameter() {parameterName=Exponent}) {materialFunction=&MF_Fresnel_FX}
  MaterialFunctionCall(Emissive=Multiply(A=@n1.RGB, B=@n2.Fresnel)) {materialFunction=&MF_FX_EmissiveAdjustments}
  n3 = Multiply(A=@n1.A, B=Multiply(A=@n2.Fresnel, B=VertexColor()))
  QualitySwitch(Default=DepthFade(Opacity=@n3, FadeDistance=ScalarParameter() {defaultValue=2, parameterName=DepthFade}), Low=@n3, Medium=@n3)
 ]
}
