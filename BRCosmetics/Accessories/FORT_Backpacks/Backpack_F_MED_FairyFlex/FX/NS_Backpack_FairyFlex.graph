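# Niagara system for the FairyFlex back bling cosmetic, adding persistent glows to the backpack in six color variants.
# Front, side and top-down sprites use fixed facings around it, alongside a mesh crevice glow and socket-attached eye sprites.
# User.Variant indexes color and alpha arrays, with time-of-day scaling and quality level adjusting some opacity and camera offsets.
# Eye sprites pass their socket's up vector to the material through dynamic parameters.
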NiagaraSystem /BRCosmetics/Accessories/FORT_Backpacks/Backpack_F_MED_FairyFlex/FX/NS_Backpack_FairyFlex
dependencies {
 usedDataInterfaces [
  &NiagaraDataInterfaceArrayColor
  &NiagaraDataInterfaceArrayFloat
  &NiagaraDataInterfaceArrayFloat3
  &NiagaraDataInterfaceSkeletalMesh
 ]
 usedDynamicInputs []
 usedModules [
  &self:SystemSpawnScript.SystemScriptSource.SystemScriptGraph.NiagaraNodeAssignment_0.SetVariables_#1
  &SystemState
 ]
}
emitters {
 FrontGlow {
  emitterUpdateScript [
   &EmitterState(
    Inactive Response=::NewEnumerator1
    Life Cycle Mode=::NewEnumerator1
    Loop Behavior=::NewEnumerator1
    Loop Duration Mode=::NewEnumerator1
    UseLoopDelay=false
   )
   &SpawnBurst_Instantaneous(Use Spawn Probability=false, UseLoopCountLimit=false)
  ]
  particleSpawnScript [
   &InitializeParticle(Lifetime=5, Sprite Size Mode=::NewEnumerator3, Uniform Sprite Size=18, UsePositionOffset=false, Write Material Random=true)
   &SpriteFacingAndAlignment(Sprite Facing=<0 1 0>, WriteAlignment=false, WriteFacing=true)
  ]
  particleUpdateScript [
   &ParticleState(Kill Particles When Lifetime Has Elapsed=false)
   &SystemLocation(MaskSpawnGroup=false, Offset=<0 -9 0>)
   &CameraOffset(Camera Offset Amount=4) {module=CameraOffset001}
   &ScaleColor(
    Scale Alpha=&Multiply_Float(
     A=&Multiply_Float(
      A=&SelectFloatFromArray(Direct Array Index=$Engine.QualityLevel, Float Selection Array=DataInterface() {propertyValues {FloatData [0.5,0.5,1,1,0]}})
      B=&SelectFloatFromArray(
       Direct Array Index=&MakeIntFromFloat(Float=$User.Variant)
       Float Selection Array=DataInterface() {propertyValues {FloatData [0.5,0.5,0.35,0.5,0.5,0.3]}}
      )
     )
     B=&FNTimeOfDayScalar(Night Scale=0.6, Morning Scale=0.8, Evening Scale=0.8)
    )
    Scale RGB=&MakeVectorFromLinearColorRGB(Color=&SelectLinearColorFromArray(
      Array Sampling Mode=::NewEnumerator1
      Direct Array Index=&MakeIntFromFloat(Float=$User.Variant)
      Color Selection Array=DataInterface() {propertyValues {ColorData [
         <rgba 2 0.2087 1.9064 1>
         <rgba 0.7163 1.5 0.1565 1>
         <rgba 1 0.65 0 1>
         <rgba 0 1.2861 2 0.5>
         <rgba 0.145 1 0.6838 0.5>
         <rgba 0 2.3333 4 1>
        ]}}
      Random Seed=?
     ))
    ScaleA=true
    ScaleRGB=true
   ) {module=ScaleColor002}
  ]
  rendererClasses [&NiagaraSpriteRendererProperties]
  renderers [{rendererClass=&NiagaraSpriteRendererProperties}]
 }
 MeshCrevicesGlow {
  emitterUpdateScript [
   &EmitterState(
    Inactive Response=::NewEnumerator1
    Life Cycle Mode=::NewEnumerator1
    Loop Behavior=::NewEnumerator1
    Loop Duration Mode=::NewEnumerator1
    UseLoopDelay=false
   )
   &SpawnBurst_Instantaneous(Use Spawn Probability=false, UseLoopCountLimit=false)
  ]
  particleSpawnScript [
   &InitializeParticle(Lifetime=5, Sprite Size Mode=::NewEnumerator3, Uniform Sprite Size=90, UsePositionOffset=false, Write Material Random=true)
   &CameraOffset(Camera Offset Amount=1)
  ]
  particleUpdateScript [
   &ParticleState(Kill Particles When Lifetime Has Elapsed=false)
   &ScaleColor(
    Scale Alpha=&Multiply_Float(
     A=&Multiply_Float(
      A=0.5
      B=&SelectFloatFromArray(
       Direct Array Index=&MakeIntFromFloat(Float=$User.Variant)
       Float Selection Array=DataInterface() {propertyValues {FloatData [0.5,0.5,0.35,0.5,0.5,0.3]}}
      )
     )
     B=&FNTimeOfDayScalar(Night Scale=1, Morning Scale=0.8, Evening Scale=0.8)
    )
    Scale RGB=&MakeVectorFromLinearColorRGB(Color=&SelectLinearColorFromArray(
      Array Sampling Mode=::NewEnumerator1
      Direct Array Index=&MakeIntFromFloat(Float=$User.Variant)
      Color Selection Array=DataInterface() {propertyValues {ColorData [<rgba 3 0 2.1712 1>, <rgba 0.625 1.5 0 1>, <rgba 2 1.3 0 1>, <rgba 0 3.2152 5 0.5>, <rgba 0.29 2 1.3676 0.5>, <rgba 0 2.3333 4 1>]}}
      Random Seed=?
     ))
    ScaleA=true
    ScaleRGB=true
   ) {module=ScaleColor002}
  ]
  rendererClasses [&NiagaraMeshRendererProperties]
  renderers [{rendererClass=&NiagaraMeshRendererProperties, rendererIndex=0}]
 }
 NE_EyesGlow {
  emitterUpdateScript [
   &EmitterState(
    Inactive Response=::NewEnumerator1
    Life Cycle Mode=::NewEnumerator1
    Loop Behavior=::NewEnumerator1
    Loop Duration Mode=::NewEnumerator1
    UseLoopDelay=false
   )
   &SpawnBurst_Instantaneous(Spawn Count=4, Use Spawn Probability=false, UseLoopCountLimit=false)
  ]
  particleSpawnScript [
   &FN_SetSpawnGroup()
   &InitializeParticle(
    Color=<rgba 10 10 10 1>
    Lifetime=5
    Position Offset=<0 0.1 0>
    Sprite Size Mode=::NewEnumerator3
    Uniform Sprite Size=&SelectFloatFromArray(Direct Array Index=$Particles.SpawnGroup, Float Selection Array=DataInterface() {propertyValues {FloatData [4,4,2.5,2.5]}})
    UsePositionOffset=false
    Write Material Random=true
   )
  ]
  particleUpdateScript [
   &ParticleState(Kill Particles When Lifetime Has Elapsed=false)
   &SkeletalMeshLocation(
    Bone / Socket Index=$Particles.SpawnGroup
    MaskSpawnGroup=false
    Mesh Sampling Type=::NewEnumerator5
    Sampled Position Offset=&SelectVectorFromArray(
     Direct Array Index=$Particles.SpawnGroup
     Vector Selection Array=DataInterface() {propertyValues {FloatData [<0 0 0>, <0 0 0>, <0 0 0>, <0 0 0>]}}
    )
    Skeletal Mesh=$System.Skeletal Mesh
    Socket Sampling Mode=::NewEnumerator5
   )
   &ScaleColor(
    Scale Alpha=&Multiply_Float(
     A=&Multiply_Float(
      A=0.3
      B=&SelectFloatFromArray(
       Direct Array Index=&MakeIntFromFloat(Float=$User.Variant)
       Float Selection Array=DataInterface() {propertyValues {FloatData [0.5,0.5,0.35,0.5,0.5,0.3]}}
      )
     )
     B=&FNTimeOfDayScalar(Night Scale=0.6, Morning Scale=0.8, Evening Scale=0.8)
    )
    Scale RGB=&MakeVectorFromLinearColorRGB(Color=&SelectLinearColorFromArray(
      Array Sampling Mode=::NewEnumerator1
      Direct Array Index=&MakeIntFromFloat(Float=$User.Variant)
      Color Selection Array=DataInterface() {propertyValues {ColorData [
         <rgba 0.4174 0.1491 0.7454 0.1>
         <rgba 0.2917 0.7 0 0.2>
         <rgba 1.5 0.7701 0.2142 1>
         <rgba 0 1.2861 2 0.5>
         <rgba 0.29 2 1.3676 0.5>
         <rgba 0 0.5833 1 1>
        ]}}
      Random Seed=?
     ))
    ScaleA=true
    ScaleRGB=true
   ) {module=ScaleColor002}
   &DynamicMaterialParameters(
    Index 0 Param 1=&MakeFloatFromVector(
     Vector=&RotateVectorByQuaternion(Vector To Rotate=<0 0 1>, Quaternion=$Particles.SkeletalMeshLocation.SampledRotation)
     Channel=::NewEnumerator0
    )
    Index 0 Param 2=&MakeFloatFromVector(
     Vector=&RotateVectorByQuaternion(Vector To Rotate=<0 0 1>, Quaternion=$Particles.SkeletalMeshLocation.SampledRotation)
     Channel=::NewEnumerator1
    )
    Index 0 Param 3=&MakeFloatFromVector(
     Vector=&RotateVectorByQuaternion(Vector To Rotate=<0 0 1>, Quaternion=$Particles.SkeletalMeshLocation.SampledRotation)
     Channel=::NewEnumerator2
    )
    Index 0 Param 4=3
    Index 1 Param 2=30
    Module.Index0Param0Write=true
    Module.Index0Param1Write=true
    Module.Index0Param2Write=true
    Module.Index0Param3Write=true
    Module.Index1Param0Write=true
    Module.Index1Param1Write=true
    Module.Index1Param2Write=true
    Module.Index1Param3Write=true
    Write Parameter Index 1=true
   ) {module=DynamicMaterialParameters001}
  ]
  rendererClasses [&NiagaraSpriteRendererProperties]
  renderers [{rendererClass=&NiagaraSpriteRendererProperties}]
 }
 SideGlow {
  emitterUpdateScript [
   &EmitterState(
    Inactive Response=::NewEnumerator1
    Life Cycle Mode=::NewEnumerator1
    Loop Behavior=::NewEnumerator1
    Loop Duration Mode=::NewEnumerator1
    UseLoopDelay=false
   )
   &SpawnBurst_Instantaneous(Use Spawn Probability=false, UseLoopCountLimit=false)
  ]
  particleSpawnScript [
   &InitializeParticle(Lifetime=5, Sprite Size=<18 20.7002>, Sprite Size Mode=::NewEnumerator4, UsePositionOffset=false, Write Material Random=true)
   &SpriteFacingAndAlignment(WriteAlignment=false, WriteFacing=true)
  ]
  particleUpdateScript [
   &ParticleState(Kill Particles When Lifetime Has Elapsed=false)
   &SystemLocation(MaskSpawnGroup=false, Offset=<0 -9 0>)
   &CameraOffset(Camera Offset Amount=&SelectFloatFromArray(Direct Array Index=$Engine.QualityLevel, Float Selection Array=DataInterface() {propertyValues {FloatData [5,5,7,7,7]}}))
   &ScaleColor(
    Scale Alpha=&Multiply_Float(
     A=&Multiply_Float(
      A=1
      B=&SelectFloatFromArray(
       Direct Array Index=&MakeIntFromFloat(Float=$User.Variant)
       Float Selection Array=DataInterface() {propertyValues {FloatData [0.5,0.5,0.5,0.5,0.5,0.5]}}
      )
     )
     B=&FNTimeOfDayScalar(Night Scale=0.6, Morning Scale=0.8, Evening Scale=0.8)
    )
    Scale RGB=&MakeVectorFromLinearColorRGB(Color=&SelectLinearColorFromArray(
      Array Sampling Mode=::NewEnumerator1
      Direct Array Index=&MakeIntFromFloat(Float=$User.Variant)
      Color Selection Array=DataInterface() {propertyValues {ColorData [
         <rgba 1 0.1043 0.9532 1>
         <rgba 0.3582 0.75 0.0783 1>
         <rgba 0.75 0.4875 0 1>
         <rgba 0 0.4823 0.75 0.5>
         <rgba 0.1088 0.75 0.5128 0.5>
         <rgba 0 0.4375 0.75 1>
        ]}}
      Random Seed=?
     ))
    ScaleA=true
    ScaleRGB=true
   ) {module=ScaleColor002}
  ]
  rendererClasses [&NiagaraSpriteRendererProperties]
  renderers [{rendererClass=&NiagaraSpriteRendererProperties}]
 }
 TopDownGlow {
  emitterUpdateScript [
   &EmitterState(
    Inactive Response=::NewEnumerator1
    Life Cycle Mode=::NewEnumerator1
    Loop Behavior=::NewEnumerator1
    Loop Duration Mode=::NewEnumerator1
    UseLoopDelay=false
   )
   &SpawnBurst_Instantaneous(Use Spawn Probability=false, UseLoopCountLimit=false)
  ]
  particleSpawnScript [
   &InitializeParticle(Lifetime=5, Sprite Size=<23 15>, Sprite Size Mode=::NewEnumerator4, UsePositionOffset=false, Write Material Random=true)
   &SpriteFacingAndAlignment(Sprite Facing=<0 0 1>, WriteAlignment=false, WriteFacing=true)
  ]
  particleUpdateScript [
   &ParticleState(Kill Particles When Lifetime Has Elapsed=false)
   &SystemLocation(MaskSpawnGroup=false, Offset=<0 -10 0>)
   &CameraOffset(Camera Offset Amount=&SelectFloatFromArray(Direct Array Index=$Engine.QualityLevel, Float Selection Array=DataInterface() {propertyValues {FloatData [0,0,8,8,8]}})) {module=CameraOffset001}
   &ScaleColor(
    Scale Alpha=&Multiply_Float(
     A=&Multiply_Float(
      A=1
      B=&SelectFloatFromArray(
       Direct Array Index=&MakeIntFromFloat(Float=$User.Variant)
       Float Selection Array=DataInterface() {propertyValues {FloatData [0.5,0.5,0.5,0.5,0.5,0.5]}}
      )
     )
     B=&FNTimeOfDayScalar(Night Scale=0.6, Morning Scale=0.8, Evening Scale=0.8)
    )
    Scale RGB=&MakeVectorFromLinearColorRGB(Color=&SelectLinearColorFromArray(
      Array Sampling Mode=::NewEnumerator1
      Direct Array Index=&MakeIntFromFloat(Float=$User.Variant)
      Color Selection Array=DataInterface() {propertyValues {ColorData [
         <rgba 1 0.1043 0.9532 1>
         <rgba 0.3582 0.75 0.0783 1>
         <rgba 0.75 0.4875 0 1>
         <rgba 0 0.4823 0.75 0.5>
         <rgba 0.1088 0.75 0.5128 0.5>
         <rgba 0 0.4375 0.75 1>
        ]}}
      Random Seed=?
     ))
    ScaleA=true
    ScaleRGB=true
   )
  ]
  rendererClasses [&NiagaraSpriteRendererProperties]
  renderers [{rendererClass=&NiagaraSpriteRendererProperties}]
 }
}
summary {
 systemName=NS_Backpack_FairyFlex
 emitters [
  {emitterName=MeshCrevicesGlow, rendererClasses [&NiagaraMeshRendererProperties]}
  {emitterName=FrontGlow, rendererClasses [&NiagaraSpriteRendererProperties]}
  {emitterName=NE_EyesGlow, rendererClasses [&NiagaraSpriteRendererProperties]}
  {emitterName=SideGlow, rendererClasses [&NiagaraSpriteRendererProperties]}
  {emitterName=TopDownGlow, rendererClasses [&NiagaraSpriteRendererProperties]}
 ]
 userVariables [
  User.bIsFrontEnd {type=NiagaraBool}
  User.bIsFrontEndPreview {type=NiagaraBool}
  User.Variant {type=NiagaraFloat, defaultValue=0, description=""}
 ]
}
